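; Healing orb pickup that heals the first player-controlled pawn to touch it, then spirals into that pawn.
; The server launches it in a random horizontal direction with upward lift and stops its projectile movement after a bounce limit.
; A timer enables pickup overlaps after a delay; overlap plays a sound, applies the heal GameplayEffect and enables tick.
; Tick orbits the target with power-curve shrinking radius and scale, destroying the orb when lifetime ends.
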(event EventBeginPlay
  (Actor|Tick|SetActorTickEnabled false)
  (Utilities|Time|SetTimerbyEvent (delegate Custom|Activate) (Variables|Default|GetDelayBeforeActivation))
  (Networking|SwitchHasAuthority
    (:Authority
      (Projectile|SetVelocity (Variables|Default|GetFortProjectileMovement) (+ (* (Math|Vector|GetForwardVector (Math|Rotator|MakeRotator 0.0 0.0 (Math|Random|RandomFloatinRange 0.0 3.6e+02))) (Variables|Default|GetLaunchSpeed)) (Math|Vector|MakeVector 0.0 0.0 (Variables|Default|GetLaunchUpSpeed)))))))

(event Custom|Activate
  (Variables|/Script/Engine.PrimitiveComponent|SetbGenerateOverlapEvents true (Variables|Default|GetFindPlayerCollision)))

(event OnComponentBeginOverlap(FindPlayerCollision)
  (params (OverlappedComponent :type Primitive%20Component%20Object%20Reference) (OtherActor :type Actor%20Object%20Reference) (OtherComp :type Primitive%20Component%20Object%20Reference) (OtherBodyIndex :type Integer) (bFromSweep :type Boolean) (SweepResult :type Hit%20Result%20Structure%20%28by%20ref%29))
  (bind _asfort_player_pawn (Utilities|Casting|CastToFortPlayerPawn OtherActor))
  (if (Pawn|IsPlayerControlled _asfort_player_pawn)
    (Utilities|FlowControl|DoOnce false
      (:Completed
        (Audio|PlaySoundatLocation "/Game/Sounds/Fort_GamePlay_Sounds/Loot/EatHealth_Restore_Athena_Cue.EatHealth_Restore_Athena_Cue" (Transformation|GetActorLocation))
        (Environment|ApplyGlobalEnvironmentGameplayEffecttoActor _asfort_player_pawn "/DentureSay/Gameplay/HealingOrb/GE_DentureSay_HealingOrb_Heal.GE_DentureSay_HealingOrb_Heal_C")
        (if (not (Variables|Default|GetbActive))
          (Actor|Tick|SetActorTickEnabled true)
          (Variables|Default|SetTargetPawn _asfort_player_pawn)
          (Variables|Default|SetbActive true)
          (Variables|Default|SetElapsed 0.0))))))

(event EventTick
  (params (DeltaSeconds :type Float%20%28single-precision%29))
  (bind _returnvalue 1.0)
  (bind _radiusexp (Variables|Default|GetRadiusExp))
  (bind _lifetime (Variables|Default|GetLifetime))
  (bind _elapsed (Variables|Default|GetElapsed))
  (bind _returnvalue_2 (+ _elapsed DeltaSeconds))
  (bind _returnvalue_3 (/ _returnvalue_2 _lifetime))
  (bind _returnvalue_4 (- _returnvalue _returnvalue_3))
  (bind _returnvalue_5 (Math|Float|Power _returnvalue_4 _radiusexp))
  (bind _self self)
  ; Make the orb rotate around the player in a spiral
  (if (Variables|Default|GetbActive)
    (Utilities|IsValid (Variables|Default|GetTargetPawn)
      (:Is%20Valid
        (Variables|Default|SetElapsed _returnvalue_2)
        (if (>= _returnvalue_3 1.0)
          (Actor|DestroyActor self)
          (else
            (Transformation|SetActorLocation _self (+ (+ (Transformation|GetActorLocation (Variables|Default|GetTargetPawn)) (Math|Vector|MakeVector 0.0 0.0 (Variables|Default|GetHeightOffset))) (Math|Vector|MakeVector (* (Math|Trig|Cos(Degrees) (+ (Variables|Default|GetStartAngleDeg) (* (* 3.6e+02 (Variables|Default|GetRevolutions)) _returnvalue_3))) (* (Variables|Default|GetStartRadius) _returnvalue_5)) (* (Math|Trig|Sin(Degrees) (+ (Variables|Default|GetStartAngleDeg) (* (* 3.6e+02 (Variables|Default|GetRevolutions)) _returnvalue_3))) (* (Variables|Default|GetStartRadius) _returnvalue_5)))) false false)
            (Transformation|SetActorScale3D _self (Math|Vector|MakeVector (* (Variables|Default|GetStartScale) _returnvalue_5) (* (Variables|Default|GetStartScale) _returnvalue_5) (* (Variables|Default|GetStartScale) _returnvalue_5))))))
      (:Is%20Not%20Valid
        (Actor|DestroyActor _self)))))

(event Collision|EventHit
  (params (MyComp :type Primitive%20Component%20Object%20Reference) (Other :type Actor%20Object%20Reference) (OtherComp :type Primitive%20Component%20Object%20Reference) (bSelfMoved :type Boolean) (HitLocation :type Vector :default "0, 0, 0") (HitNormal :type Vector :default "0, 0, 0") (NormalImpulse :type Vector :default "0, 0, 0") (Hit :type Hit%20Result%20Structure%20%28by%20ref%29))
  (bind _returnvalue 1)
  (bind _bouncecount (Variables|Default|GetBounceCount))
  (bind _returnvalue_2 (+ _bouncecount _returnvalue))
  ; SERVER ONLY: the server owns the bounce simulation; clients are corrected
  ; via Replicate Movement. Bounce handling only matters during the launch
  ; phase - once the spiral has grabbed the orb (bActive) or it has settled, ignore hits.
  (Networking|SwitchHasAuthority
    (:Authority
      (if (and (not (Variables|Default|GetbActive)) (not (Variables|Default|GetbSettled)))
        (Variables|Default|SetBounceCount _returnvalue_2)
        (if (>= _returnvalue_2 (Variables|Default|GetMaxBounces))
          (Components|Movement|StopMovementImmediately (Variables|Default|GetFortProjectileMovement))
          (Components|Activation|Deactivate (Variables|Default|GetFortProjectileMovement))
          (Variables|Default|SetbSettled true))))))

(fn ConstructionScript)
